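# Additive, unlit, two-sided mesh-particle material for barrier light cones in the Butterfly effects.
# A flippable UV1 gradient times a view-angle fade shapes opacity and blends emissive from a parameter color to particle color, boosted through DeriveHDRfromLDR.
# Panning fog noise, offset by actor position and masked by a second panning texture, breaks up opacity before depth fade.
# Low quality skips noise and depth fade; particle alpha scales both paths.
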Material /Figment_S10/Athena/Effects/Materials/Butterfly/M_Barrier_LightCones
graph {
 properties {
  bUsedWithMeshParticles=true
  blendMode=BLEND_Additive
  pixelDepthOffsetMode=PDOM_Legacy
  shadingModel=MSM_Unlit
  twoSided=true
  usedShadingModels=Unlit
 }
 nodes [
  n1 = VertexNormalWS()
  n2 = ComponentMask(None=TextureCoordinate() {coordinateIndex=1}) {g=true}
  LinearInterpolate(A=Multiply(A=ScalarParameter() {defaultValue=1, parameterName=A}, B=@n1), B=@n1, Alpha=Clamp(None=Power(Base=@n2) {constExponent=4}))
  n3 = VectorParameter() {parameterName=LightEmissiveColor}
  n4 = ParticleColor()
  n5 = ComponentMask(None=TextureCoordinate() {coordinateIndex=1}) {g=true}
  n6 = Power(
   Base=LinearInterpolate(A=@n5, B=OneMinus(None=@n5), Alpha=ScalarParameter() {parameterName=FlipY})
   Exp=ScalarParameter() {defaultValue=1, parameterName=ConePow}
  )
  n7 = MaterialFunctionCall("FadeSensitivity (optional)"=ScalarParameter() {defaultValue=0.7, parameterName=FadeSensitivity}) {materialFunction=&FadeBasedOnViewAngle}
  n8 = Clamp(None=Multiply(A=@n6, B=@n7.Result))
  MaterialFunctionCall(
   LDR Input=Clamp(None=LinearInterpolate(A=@n3.RGB, B=@n4.RGB, Alpha=@n8))
   HDR Tint=@n4.RGB
   Desaturate Luminance=ScalarParameter() {defaultValue=0.15, parameterName=HDRLum}
  ) {materialFunction=&DeriveHDRfromLDR}
  n9 = Multiply(A=@n6, B=@n7.Result)
  n10 = ActorPositionWS()
  n11 = Add(
   A=TextureCoordinate() {uTiling=5}
   B=Multiply(A=ComponentMask(None=@n10) {g=true, r=true}, B=ScalarParameter() {defaultValue=0.0001, parameterName=adsfsaf}) {constB=0.005}
  )
  n12 = ScalarParameter() {defaultValue=5, parameterName=TileFogNoise}
  n13 = TextureSample(UVs=Panner(Coordinate=Multiply(A=@n11, B=@n12)) {speedY=0.05}) {texture=/Game/Effects/Fort_Effects/Textures/T_Smooth_tile}
  n14 = TextureSample(UVs=Panner(Coordinate=TextureCoordinate()) {speedX=-0.125}) {texture=/Game/Effects/Fort_Effects/Textures/T_HitImpact}
  n15 = Multiply(
   A=ScalarParameter() {defaultValue=0.45, parameterName=OpacityScale}
   B=LinearInterpolate(A=@n9, B=Multiply(A=@n9, B=@n13.G), Alpha=OneMinus(None=@n14.G))
  )
  Multiply(A=ParticleColor().A, B=QualitySwitch(Default=DepthFade(Opacity=@n15), Low=Multiply(A=@n9)))
 ]
}
